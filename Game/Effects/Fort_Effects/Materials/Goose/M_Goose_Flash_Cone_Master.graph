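# Unlit translucent master material for mesh-particle flash cones in the Goose effects, showing streaming smoky light with an optional bright core.
# Panned smoke textures on power-warped, normal-distorted UVs drive desaturated, particle-tinted emissive and, with swirl noise and gradients, opacity.
# Depth, Fresnel and camera-distance fades shape opacity, with dynamic-parameter fade inversion.
# Quality and shading-path switches drop costly layers; a Debug switch, on by default, replaces particle-driven time with Time.
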Material /Game/Effects/Fort_Effects/Materials/Goose/M_Goose_Flash_Cone_Master
graph {
 properties {bUsedWithMeshParticles=true, blendMode=BLEND_Translucent, pixelDepthOffsetMode=PDOM_Legacy, shadingModel=MSM_Unlit, usedShadingModels=Unlit}
 nodes [
  n1 = MaterialFunctionCall(Float2=TextureCoordinate() {vTiling=1.3}) {materialFunction=&BreakOutFloat2Components}
  n2 = AppendVector(A=@n1.R, B=Power(Base=@n1.G) {constExponent=6})
  n3 = Multiply(
   A=StaticSwitchParameter(True=Time(), False=DynamicParameter() {paramNames [TimeMutiplier, InvertFade, Param3, Ref]}.TimeMutiplier) {defaultValue=true, parameterName=Debug}
   B=ScalarParameter() {defaultValue=1, parameterName=TimeMutiplier}
  )
  n4 = AppendVector(A=ScalarParameter() {defaultValue=1, parameterName=MasterMaskU}, B=ScalarParameter() {defaultValue=1, parameterName=MasterMaskV})
  n5 = TextureSample(UVs=Multiply(A=Panner(Coordinate=@n2, Time=@n3) {speedX=0.5, speedY=-4}, B=@n4)) {texture=/Game/Effects/Fort_Effects/Textures/Smoke/T_Smoke_Trail_Soft_Chromatic}
  n6 = MaterialFunctionCall(Float2=TextureCoordinate()) {materialFunction=&BreakOutFloat2Components}
  n7 = MaterialFunctionCall(Float2=TextureCoordinate() {uTiling=2, vTiling=0.6}) {materialFunction=&BreakOutFloat2Components}
  n8 = Power(Base=@n7.G)
  n9 = TextureSample(UVs=Panner(Coordinate=AppendVector(A=@n7.R, B=@n8), Time=@n3) {speedY=-0.5}) {samplerType=SAMPLERTYPE_Normal, texture=&T_Weap_Distort_N_02}
  n10 = Constant()
  n11 = ShadingPathSwitch(Default=Multiply(A=ComponentMask(None=@n9.RGB) {g=true, r=true}) {constB=0.12}, Mobile=@n10) {desc=Migrated from FeatureLevelSwitch}
  n12 = QualitySwitch(Default=@n11, Low=@n10, High=@n11, Medium=@n10)
  n13 = ShadingPathSwitch(Default=@n12, Deferred=@n12, Forward=@n10)
  n14 = Add(A=AppendVector(A=@n6.R, B=Power(Base=@n6.G) {constExponent=4}), B=@n13)
  n15 = TextureSample(UVs=Multiply(A=Panner(Coordinate=@n14, Time=@n3) {speedX=0.1, speedY=-1}, B=@n4)) {texture=&T_Smoke_Trail_Soft_Chromatic}
  n16 = Constant()
  n17 = ShadingPathSwitch(Default=Add(A=Multiply(A=@n5.RGB) {constB=0.35}, B=Multiply(A=@n15.RGB) {constB=0.4}), Mobile=@n16) {desc=Migrated from FeatureLevelSwitch}
  n18 = QualitySwitch(Default=@n17, Low=@n16, High=@n17, Medium=@n16)
  n19 = Panner(Coordinate=@n14, Time=@n3) {speedX=-0.2, speedY=-1}
  n20 = Add(
   A=ShadingPathSwitch(Default=@n18, Deferred=@n18, Forward=@n16)
   B=TextureSample(UVs=Multiply(A=@n19, B=@n4)) {texture=&T_Smoke_Trail_Soft_Chromatic}.RGB
  )
  n21 = MaterialFunctionCall(Float2=TextureCoordinate() {vTiling=0.7}) {materialFunction=&BreakOutFloat2Components}
  n22 = Add(A=@n13, B=AppendVector(A=@n21.R, B=Power(Base=@n21.G) {constExponent=4}))
  n23 = TextureSample(UVs=Multiply(A=Panner(Coordinate=@n22, Time=@n3) {speedX=0.2, speedY=-1}, B=@n4)) {texture=/Game/Effects/Fort_Effects/Textures/Smoke/T_Smoke_Trail_Soft_2}
  n24 = TextureCoordinate() {uTiling=2}
  n25 = LinearInterpolate(Alpha=Distance(A=ComponentMask(None=@n24) {r=true}, B=Constant() {r=1})) {constA=10, constB=0}
  n26 = Power(Base=TextureCoordinate() {uTiling=0}, Exp=ScalarParameter() {defaultValue=1, parameterName=Center_Power})
  n27 = Constant2Vector()
  n28 = ScalarParameter() {defaultValue=0.75, parameterName=Center_Radius}
  n29 = ScalarParameter() {defaultValue=0.25, parameterName=Center_Hardness}
  n30 = Add(
   A=Multiply(A=Add(A=@n20, B=@n23.RGB), B=Saturate(None=@n25))
   B=StaticSwitchParameter(True=SphereMask(A=@n26, B=@n27, Radius=@n28, Hardness=@n29), False=Constant()) {defaultValue=true, parameterName=UseBrightCore}
  )
  n31 = Desaturation(None=@n30, Fraction=Constant() {r=0.4})
  MaterialFunctionCall(Emissive=Multiply(A=ParticleColor().RGB, B=Multiply(A=@n3, B=@n31))) {materialFunction=&MF_FX_EmissiveAdjustments}
  n32 = Constant() {r=1}
  n33 = ShadingPathSwitch(
   Default=MaterialFunctionCall(Distance=ScalarParameter() {defaultValue=500, parameterName=CameraFadeDis}) {materialFunction=&Near_Camera_Fade}.Result
   Mobile=@n32
  ) {desc=Migrated from FeatureLevelSwitch}
  n34 = QualitySwitch(Default=@n33, Low=@n32, High=@n33, Medium=@n33)
  n35 = MaterialFunctionCall(Float2=TextureCoordinate()) {materialFunction=&BreakOutFloat2Components}
  n36 = Saturate(None=Multiply(A=OneMinus(None=@n35.G)))
  n37 = TextureCoordinate() {vTiling=0.5}
  n38 = Constant() {r=1}
  n39 = ShadingPathSwitch(
   Default=TextureSample(UVs=Panner(Coordinate=@n37, Time=@n3) {speedX=0.2, speedY=-1}) {samplerType=SAMPLERTYPE_LinearGrayscale, texture=/Game/Global/Textures/Default/Noise/T_SwirlNoise_N_height}.R
   Mobile=@n38
  ) {desc=Migrated from FeatureLevelSwitch}
  n40 = ShadingPathSwitch(Default=QualitySwitch(Default=@n39, Low=@n38, High=@n39, Medium=@n38), Forward=@n38)
  n41 = Multiply(A=Multiply(A=@n30, B=Multiply(A=@n36, B=@n40)), B=ScalarParameter() {defaultValue=1, parameterName=OpacityOverall})
  n42 = ScalarParameter() {defaultValue=2500, parameterName=FadeDistance}
  n43 = QualitySwitch(
   Default=ShadingPathSwitch(Default=DepthFade(Opacity=@n41, FadeDistance=@n42) {fadeDistanceDefault=110}, Mobile=@n41) {desc=Migrated from FeatureLevelSwitch}
   Low=@n41
  )
  n44 = Constant() {r=1}
  n45 = ShadingPathSwitch(Default=OneMinus(None=Fresnel() {baseReflectFraction=0, exponent=1}), Mobile=@n44) {desc=Migrated from FeatureLevelSwitch}
  n46 = QualitySwitch(Default=@n45, Low=@n44, High=@n45, Medium=@n44)
  n47 = Multiply(
   A=Multiply(A=ParticleColor().A, B=ShadingPathSwitch(Default=@n43, Forward=@n41))
   B=StaticSwitchParameter(True=ShadingPathSwitch(Default=@n46, Forward=@n44), False=Constant() {r=1}) {defaultValue=true, parameterName=UseEdgeFade}
  )
  n48 = Subtract(A=@n35.G, B=ScalarParameter() {defaultValue=0.5, parameterName=GradientMaskBias})
  n49 = Multiply(A=@n47, B=Power(Base=Saturate(None=@n48)) {constExponent=1})
  n50 = MaterialFunctionCall(Distance=ScalarParameter() {defaultValue=8192, parameterName=NearFade}) {materialFunction=&Near_Camera_Fade}
  n51 = LinearInterpolate(A=@n50.Result, B=OneMinus(None=@n50.Result), Alpha=DynamicParameter() {paramNames [TimeMutiplier, InvertFade, Param3, Ref]}.InvertFade)
  n52 = Constant() {r=0.25}
  Multiply(
   A=Multiply(
    A=ShadingPathSwitch(Default=@n34, Deferred=@n34, Mobile=@n32)
    B=StaticSwitchParameter(True=@n47, False=@n49) {defaultValue=true, parameterName=UseBrightCore}
   )
   B=Clamp(None=Subtract(A=@n51, B=@n52))
  )
 ]
}
